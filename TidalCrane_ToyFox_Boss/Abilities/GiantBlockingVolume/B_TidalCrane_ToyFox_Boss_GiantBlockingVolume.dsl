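; Giant blocking volume Blueprint for the TidalCrane ToyFox boss, built around a box component named for stopping movement.
; On BeginPlay it calls the parent, then waits asynchronously for the current playlist data to be ready.
; It then resizes the box, using one hotfixable scalable float for both X and Y extents while keeping the existing Z extent.
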(event EventBeginPlay
  (bind (_boxextent_x _boxextent_y _boxextent_z) (Math|Vector|BreakVector (Variables|/Script/Engine.BoxComponent|GetBoxExtent_X (Variables|Default|GetBox_StopMovement))))
  ;[0 0.15 0.000542 0.5] BeginPlay: Update box extent based on hotfix.
  (|CallParentFunction)
  (Class|FortAsyncActionCurrentPlaylistReady|PlaylistDataReadyAsync
    (:OnReady
      (Components|Box|SetBoxExtent (Variables|Default|GetBox_StopMovement) (Math|Vector|MakeVector (ScalableFloat|GetValueatLevel (Variables|Default|GetHotfix_BoxExtent)) (ScalableFloat|GetValueatLevel (Variables|Default|GetHotfix_BoxExtent)) _boxextent_z) false))))

(fn ConstructionScript)
