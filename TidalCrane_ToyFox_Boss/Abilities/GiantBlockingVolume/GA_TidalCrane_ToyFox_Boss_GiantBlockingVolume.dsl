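; Gameplay ability for the TidalCrane ToyFox boss that keeps a giant blocking volume in the world while the ability is active.
; On activation it clears any volume it already holds, then always spawns the blocking volume actor at the avatar's location, ignoring collision.
; The spawned actor is stored in a transient variable and destroyed when the ability ends, before the parent's end logic runs.
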(event Ability|EventActivateAbility
  (|CallParentFunction)
  (bind _blockingvolume (Variables|Transient|GetBlockingVolume)
    (:then
      (Actor|DestroyActor _blockingvolume))
    (:else
      _))
  (bind _returnvalue (Game|SpawnActorfromClass "/TidalCrane_ToyFox_Boss/Abilities/GiantBlockingVolume/B_TidalCrane_ToyFox_Boss_GiantBlockingVolume.B_TidalCrane_ToyFox_Boss_GiantBlockingVolume_C" (Math|Conversions|ToTransform(Vector) (Transformation|GetActorLocation (Ability|GetAvatarActorfromActorInfo))) "AlwaysSpawn" "MultiplyWithRoot" 0 0))
  (Variables|Transient|SetBlockingVolume _returnvalue))

(event Ability|EventOnEndAbility
  (params (bWasCancelled :type Boolean))
  (Utilities|FlowControl|Sequence
    (:then_0
      (bind _blockingvolume (Variables|Transient|GetBlockingVolume)
        (:then
          (Actor|DestroyActor _blockingvolume))
        (:else)))
    (:then_1
      (|CallParentFunction bWasCancelled))))
